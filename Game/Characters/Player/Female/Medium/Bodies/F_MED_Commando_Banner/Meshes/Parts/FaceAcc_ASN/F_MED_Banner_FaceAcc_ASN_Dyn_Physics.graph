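# Physics asset for the dynamic hair on the FaceAcc_ASN mesh part of the female medium Commando Banner player character.
# Capsule bodies cover two back-hair bones, two tail bones, and the head, neck, upper spine and left clavicle.
# Constraints form two short chains: back hair hangs from the head, while the tail attaches to the left clavicle rather than the head.
# Angular limits are tight, except for a looser twist at the tail tip.
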PhysicsAsset /Game/Characters/Player/Female/Medium/Bodies/F_MED_Commando_Banner/Meshes/Parts/FaceAcc_ASN/F_MED_Banner_FaceAcc_ASN_Dyn_Physics
bodies {
 clavicle_l [{shapeType=Capsule, length=6.75, radius=5.5, center=<6.5984 -1.17 -1.3337>, rotation=<rot -84.9986 -179.9998 179.9998>}]
 dyn_hair_bk_main_a [{shapeType=Capsule, length=2.3374, radius=2.6915, center=<1.1789 -0.1031 0>, rotation=<rot -90 0.5968 -5.5968>}]
 dyn_hair_bk_main_b [{shapeType=Capsule, length=2.25, radius=2.1063, rotation=<rot 64.9996 0 0>}]
 dyn_hair_tail_a [{shapeType=Capsule, length=1.7907, radius=1.849, center=<0 -0.0914 -0.251>, rotation=<rot 0 0 -60.0005>}]
 dyn_hair_tail_b [{shapeType=Capsule, length=0.9076, radius=0.75, center=<-0.054 -0.0677 -0.4789>, rotation=<rot 0 179.5787 -103.0531>}]
 head [{shapeType=Capsule, length=7.6105, radius=6.3147, center=<3.2707 2.4255 0.6769>, rotation=<rot 7.0992 93.3798 -92.8449>}]
 neck_01 [{shapeType=Capsule, length=4.5, radius=4, center=<1.3783 0.2959 0.0006>, rotation=<rot 90 -5.9469 -0.9469>}]
 spine_05 [{shapeType=Capsule, length=12.75, radius=9, center=<0.7226 -1.6907 0>}]
}
constraints [
 {bone1Name=dyn_hair_bk_main_a, bone2Name=head, swing1LimitDegrees=3, swing2LimitDegrees=3, twistLimitDegrees=1}
 {bone1Name=dyn_hair_bk_main_b, bone2Name=dyn_hair_bk_main_a, swing1LimitDegrees=3, swing2LimitDegrees=3, twistLimitDegrees=1}
 {bone1Name=dyn_hair_tail_a, bone2Name=clavicle_l, swing1LimitDegrees=5, swing2LimitDegrees=1, twistLimitDegrees=5}
 {bone1Name=dyn_hair_tail_b, bone2Name=dyn_hair_tail_a, swing1LimitDegrees=10, swing2LimitDegrees=1, twistLimitDegrees=30}
]
