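# Post-process style AnimBP for the Banner female commando's FaceAcc_ASN mesh, adding secondary motion and pose correctives to hair bones.
# It copies the parent mesh pose, runs RigidBody simulation, applies additive bone offsets, then AnimDynamics on two front-hair bones.
# Each update maps head and left-clavicle deltas from ref pose to 0–1 corrective alphas.
# HeadTwist_L is computed but unused, and the dyn_hair_tail_b corrective is unconnected.
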AnimBlueprint /BRCosmetics/Characters/Player/Female/Medium/Bodies/F_MED_Commando_Banner/Meshes/Parts/FaceAcc_ASN/F_MED_Banner_FaceAcc_ASN_AnimBP
target_skeleton=&/F_MED_Banner_FaceAcc_ASN_Skeleton
graphs [
 AnimGraph {nodes [
   n1 = Get ShoulderUp_L() -> (ShoulderUp_L:double)
   "Transform (Modify) Bone - Bone: dyn_hair_tail_b"(Alpha=@n1, Translation="0,0,0", Rotation="0,0,100", Scale="1,1,1") {
    rotationMode=BMM_Additive
    rotationSpace=BCS_BoneSpace
    translationMode=BMM_Additive
    translationSpace=BCS_BoneSpace
    alphaBoolBlend {blendOption=Linear}
    boneToModify {boneName=dyn_hair_tail_b}
    rotation {roll=100}
   }
   n2 = RigidBody(ComponentPose=Local To Component(LocalPose=Copy Pose From Mesh() {bUseAttachedParent=true})) {
    bClampLinearTranslationLimitToRefPose=true
    overridePhysicsAsset=/Game/Characters/Player/Female/Medium/Bodies/F_MED_Commando_Banner/Meshes/Parts/FaceAcc_ASN/F_MED_Banner_FaceAcc_ASN_Dyn_Physics
    simulationSpace=BaseBoneSpace
    alphaBoolBlend {blendOption=Linear}
    baseBoneRef {boneName=spine_04}
    componentAppliedLinearAccClamp=<1000 1000 1000>
    componentLinearAccScale=<0.3 0.3 0.3>
    componentLinearVelScale=<0.3 0.1 0.3>
   }
   n3 = "Transform (Modify) Bone - Bone: dyn_hair_bk_main_a"(ComponentPose=@n2, Alpha=Get HeadUp() -> (HeadUp:double), Translation="0,0,0", Rotation="0,-30,0", Scale="1,1,1") {
    rotationMode=BMM_Additive
    rotationSpace=BCS_BoneSpace
    translationMode=BMM_Additive
    translationSpace=BCS_BoneSpace
    alphaBoolBlend {blendOption=Linear}
    boneToModify {boneName=dyn_hair_bk_main_a}
    rotation {yaw=-30}
   }
   n4 = "Transform (Modify) Bone - Bone: dyn_hair_bk_main_a"(ComponentPose=@n3, Alpha=Get HeadDn() -> (HeadDn:double), Translation="0,0,0", Rotation="0,10,0", Scale="1,1,1") {
    rotationMode=BMM_Additive
    rotationSpace=BCS_BoneSpace
    translationMode=BMM_Additive
    translationSpace=BCS_BoneSpace
    alphaBoolBlend {blendOption=Linear}
    boneToModify {boneName=dyn_hair_bk_main_a}
    rotation {yaw=10}
   }
   n5 = "Transform (Modify) Bone - Bone: dyn_hair_bk_main_a"(ComponentPose=@n4, Alpha=Get HeadTwist_R() -> (HeadTwist_R:double), Translation="0,0,0", Rotation="-20,-10,0", Scale="1,1,1") {
    rotationMode=BMM_Additive
    rotationSpace=BCS_BoneSpace
    translationMode=BMM_Additive
    translationSpace=BCS_BoneSpace
    alphaBoolBlend {blendOption=Linear}
    boneToModify {boneName=dyn_hair_bk_main_a}
    rotation {pitch=-20, yaw=-10}
   }
   n6 = "Transform (Modify) Bone - Bone: dyn_hair_bk_main_b"(ComponentPose=@n5, Alpha=@n1, Translation="-2,0,3", Rotation="0,0,0", Scale="1,1,1") {
    rotationMode=BMM_Additive
    rotationSpace=BCS_BoneSpace
    translationMode=BMM_Additive
    translationSpace=BCS_BoneSpace
    alphaBoolBlend {blendOption=Linear}
    boneToModify {boneName=dyn_hair_bk_main_b}
    translation {x=-2, z=3}
   }
   n7 = "Transform (Modify) Bone - Bone: dyn_hair_tail_a"(ComponentPose=@n6, Alpha=@n1, Translation="5,0,1", Rotation="0,0,10", Scale="1,1,1") {
    rotationMode=BMM_Additive
    rotationSpace=BCS_BoneSpace
    translationMode=BMM_Additive
    translationSpace=BCS_BoneSpace
    alphaBoolBlend {blendOption=Linear}
    boneToModify {boneName=dyn_hair_tail_a}
    rotation {roll=10}
    translation {x=5, z=1}
   }
   n8 = Anim Dynamics - Bone: dyn_hair_front_a(ComponentPose=@n7, Alpha=1, ExternalForce="0,0,0") {
    angularDampingOverride=1
    bGravityOverrideInSimSpace=true
    bOverrideAngularDamping=true
    gravityScale=0.35
    alphaBoolBlend {blendOption=Linear}
    boundBone {boneName=dyn_hair_front_a}
    physicsBodyDefinitions [{
      sphereCollisionRadius=0
      boundBone {boneName=dyn_hair_front_a}
      boxExtents=<2 10 10>
      constraintSetup {angularLimitsMax=<0 0 20>, angularLimitsMin=<0 -20 0>, angularTarget=<0 0 0>}
      localJointOffset=<2 0 0>
     }]
   }
   n9 = Anim Dynamics - Bone: dyn_hair_front_b(ComponentPose=@n8, Alpha=1, ExternalForce="0,0,0") {
    angularDampingOverride=1
    bGravityOverrideInSimSpace=true
    bOverrideAngularDamping=true
    gravityScale=0.35
    alphaBoolBlend {blendOption=Linear}
    boundBone {boneName=dyn_hair_front_b}
    physicsBodyDefinitions [{
      sphereCollisionRadius=0
      boundBone {boneName=dyn_hair_front_b}
      boxExtents=<2 10 10>
      constraintSetup {angularLimitsMax=<0 0 40>, angularLimitsMin=<0 -20 -20>, angularTarget=<0 0 0>}
      localJointOffset=<2 0 0>
     }]
   }
   Output Pose(Result=Component To Local(ComponentPose=@n9)) {layerGroup=None}
  ]}
 EventGraph {nodes [
   Event BlueprintUpdateAnimation()
   > n1 = Sequence!(execute) -> (then_0, then_1, then_2:Exec, then_3:Exec, then_4:Exec)
   n2 = GetOwnerSkeletalMeshComponent()
   n3 = BreakRotator(InRot=BreakTransform(InTransform=GetDeltaTransformFromRefPose(self=@n2, BoneName=head, BaseName=spine_05)).Rotation)
   > Set HeadDn(HeadDn:double=MapRangeClamped(Value=@n3.Yaw, InRangeA=17, InRangeB=0, OutRangeA=1, OutRangeB=0)) -> (Output_Get:double)
   Set HeadUp(execute=@n1.then_1, HeadUp:double=MapRangeClamped(Value=@n3.Yaw, InRangeA=0, InRangeB=-44)) -> (Output_Get:double)
   n4 = BreakRotator(InRot=BreakTransform(InTransform=GetDeltaTransformFromRefPose(self=@n2, BoneName=head, BaseName=neck_01)).Rotation)
   Set HeadTwist_R(execute=@n1.then_2, HeadTwist_R:double=MapRangeClamped(Value=@n4, InRangeA=0, InRangeB=-25)) -> (Output_Get:double)
   Set HeadTwist_L(execute=@n1.then_3, HeadTwist_L:double=MapRangeClamped(Value=@n4, InRangeA=0, InRangeB=25)) -> (Output_Get:double)
   n5 = BreakTransform(InTransform=GetDeltaTransformFromRefPose(self=@n2, BoneName=clavicle_l, BaseName=spine_05))
   Set ShoulderUp_L(execute=@n1.then_4, ShoulderUp_L:double=MapRangeClamped(Value=BreakRotator(InRot=@n5.Rotation).Pitch, InRangeA=0, InRangeB=45)) -> (Output_Get:double)
   TryGetPawnOwner()
  ]}
]
variables [
 HeadDn {type=number}
 HeadTwist_L {type=number}
 HeadTwist_R {type=number}
 HeadUp {type=number}
 ShoulderUp_L {type=number}
]
